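# One-shot quad MetaSound for the Jethro ClearShell bullet whip, playing a random pass-by wave with pitch and volume variation.
# A decay envelope crossfades it from front to side channels over the pass-by time and curve, while alignment weights left against right.
# A quad panner applies azimuth and distance-driven gain.
# On stop, a distance check against the ignore-impact threshold can trigger a short fade-out that also finishes the source.
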MetaSoundSource /JethroWeaponGameplay/Assets/Audio/ClearShell/IP_MSS_Jethro_ClearShell_BulletWhip
inputs [
 BulletWhip.IgnoreImpactDistance {type=Float}
 BulletWhip.OnStop {type=Trigger}
 BulletWhip.PassByAlignment {type=Float}
 BulletWhip.PassByAzimuth {type=Float}
 BulletWhip.PassByCrossfadeTime {type=Time}
 BulletWhip.PassByCurvePower {type=Float}
 BulletWhip.PassByDistance {type=Float}
 Sounds {type=WaveAsset:Array, access=Constructor}
 UE.Source.OnPlay {type=Trigger}
]
interfaces [Fort.BulletWhip:1.0, UE.OutputFormat.Quad:1.0, UE.Source.OneShot:1.0, UE.Source:1.1]
outputs [
 "UE.OutputFormat.Quad.Audio:0" {type=Audio}
 "UE.OutputFormat.Quad.Audio:1" {type=Audio}
 "UE.OutputFormat.Quad.Audio:4" {type=Audio}
 "UE.OutputFormat.Quad.Audio:5" {type=Audio}
 UE.Source.OneShot.OnFinished {type=Trigger}
]
pages [Default {
  input_defaults {
   BulletWhip.IgnoreImpactDistance=0.1
   BulletWhip.OnStop=null
   BulletWhip.PassByAlignment=0
   BulletWhip.PassByAzimuth=-90
   BulletWhip.PassByCrossfadeTime=0.2
   BulletWhip.PassByCurvePower=4
   BulletWhip.PassByDistance=1
   UE.Source.OnPlay=null
   Sounds [
    &/IP_Jethro_ClearShell_Whip_By_01
    &/IP_Jethro_ClearShell_Whip_By_02
    &/IP_Jethro_ClearShell_Whip_By_03
    &/IP_Jethro_ClearShell_Whip_By_04
   ]
  }
  nodes [
   n1 = &MSS_PlayRandomSound_Mono(Sounds=^Sounds, UE.Source.OnPlay=^UE.Source.OnPlay) {also {defaults {MaxPitch=1.2, MaxVolume=1, MinPitch=0.9, MinVolume=0.9}}}
   ADEnvelopeADEnvelope_1 = AD Envelope.AD Envelope:Float(Trigger=TriggerCompare.Clamp:Float(B=^BulletWhip.IgnoreImpactDistance, Compare=^BulletWhip.OnStop, A=^BulletWhip.PassByDistance) {also {defaults {Type=5}}}.True) {also {defaults {Attack Time=0, Decay Time=0.1}}}
   TriggerRouteTriggerRouteFloat2_0 = "TriggerRoute.Trigger Route (Float, 2):Float"(Set 0=^UE.Source.OnPlay, Set 1=@ADEnvelopeADEnvelope_1.On Trigger, Value 1=@ADEnvelopeADEnvelope_1.Out Envelope) {also {defaults {Value 0=1}}}
   n2 = VariableAccessor.Audio(Variable=VariableMutator.Audio(
     Variable=InitVariable.Audio().Variable
     Value=UE.Multiply:Audio by Float(PrimaryOperand=@n1."UE.OutputFormat.Mono.Audio:0", AdditionalOperands=@TriggerRouteTriggerRouteFloat2_0.Value).Out
    ).Variable)
   n3 = VariableAccessor.Audio(Variable=@n2.Variable)
   ADEnvelopeADEnvelope_0 = AD Envelope.AD Envelope:Float(Decay Time=^BulletWhip.PassByCrossfadeTime, Decay Curve=^BulletWhip.PassByCurvePower, Trigger=^UE.Source.OnPlay) {also {defaults {Attack Time=0}}}
   n4 = VariableAccessor.Float(Variable=VariableMutator.Float(Value=@ADEnvelopeADEnvelope_0.Out Envelope, Variable=InitVariable.Float().Variable).Variable)
   n5 = VariableAccessor.Float(Variable=@n4.Variable)
   n6 = VariableAccessor.Float(Variable=InitVariable.Float().Variable)
   n7 = VariableAccessor.Float(Variable=VariableMutator.Float(
     Variable=InitVariable.Float().Variable
     Value=MapRange.MapRange:Float(In=^BulletWhip.PassByAlignment, In Range B=@n6.Value) {also {defaults {Out Range A=1, Out Range B=0}}}.Out Value
    ).Variable)
   n8 = UE.Multiply:Float(PrimaryOperand=@n6.Value) {also {defaults {AdditionalOperands=-1}}}
   n9 = VariableAccessor.Float(Variable=VariableMutator.Float(
     Variable=InitVariable.Float().Variable
     Value=MapRange.MapRange:Float(In=^BulletWhip.PassByAlignment, In Range B=@n8.Out) {also {defaults {Out Range A=1, Out Range B=0}}}.Out Value
    ).Variable)
   n10 = VariableAccessor.Audio(Variable=@n3.Variable)
   n11 = VariableAccessor.Audio(Variable=@n10.Variable)
   n12 = VariableAccessor.Float(Variable=VariableMutator.Float(
     Variable=InitVariable.Float().Variable
     Value=UE.Subtract:Float(AdditionalOperands=@ADEnvelopeADEnvelope_0.Out Envelope) {also {defaults {PrimaryOperand=1}}}.Out
    ).Variable)
   n13 = VariableAccessor.Float(Variable=@n12.Variable)
   MSP_AzimuthPannerQuad_0 = &MSP_AzimuthPannerQuad(
    Azimuth=^BulletWhip.PassByAzimuth
    Gain=^BulletWhip.PassByDistance
    In Front Right=UE.Multiply:Audio by Float(PrimaryOperand=UE.Multiply:Audio by Float(PrimaryOperand=@n3.Value, AdditionalOperands=@n5.Value).Out, AdditionalOperands=@n7.Value).Out
    In Front Left=UE.Multiply:Audio by Float(PrimaryOperand=UE.Multiply:Audio by Float(PrimaryOperand=@n2.Value, AdditionalOperands=@n4.Value).Out, AdditionalOperands=@n9.Value).Out
    In Side Right=UE.Multiply:Audio by Float(PrimaryOperand=UE.Multiply:Audio by Float(PrimaryOperand=@n11.Value, AdditionalOperands=@n13.Value).Out, AdditionalOperands=@n7.Value).Out
    In Side Left=UE.Multiply:Audio by Float(PrimaryOperand=UE.Multiply:Audio by Float(PrimaryOperand=@n10.Value, AdditionalOperands=@n12.Value).Out, AdditionalOperands=@n9.Value).Out
   )
   TriggerAnyTriggerAny2_0 = "TriggerAny.Trigger Any (2)"(In 1=@ADEnvelopeADEnvelope_1.On Done, In 0=@n1.UE.Source.OneShot.OnFinished)
  ]
  links [
   {dst="@UE.OutputFormat.Quad.Audio:0", src=MSP_AzimuthPannerQuad_0.Out Front Left}
   {dst="@UE.OutputFormat.Quad.Audio:1", src=MSP_AzimuthPannerQuad_0.Out Front Right}
   {dst="@UE.OutputFormat.Quad.Audio:4", src=MSP_AzimuthPannerQuad_0.Out Side Left}
   {dst="@UE.OutputFormat.Quad.Audio:5", src=MSP_AzimuthPannerQuad_0.Out Side Right}
   {dst="@UE.Source.OneShot.OnFinished", src=TriggerAnyTriggerAny2_0.Out}
  ]
 }]
properties {
 assetImportData=&self:AssetImportData
 numChannels=4
 outputFormat=Quad
 pitch=1.2
 sampleRate=48000
 soundClassObject=&Fort_WeaponRanged_Projectile_Effects
 volume=0.9
 concurrencySet [/Game/Sounds/Concurrency_Rules/Athena_BulletWhizBy_Conc]
}
