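# Sound cue for the overheat foley of Jethro's ClearShell weapon.
# A mixer layers the main overheat wave with an end wave that starts after a fixed delay of about 2.3 seconds at reduced volume.
# It uses weapon reload attenuation and the weapon foley sound class.
# A concurrency override allows one instance per owner and blocks new plays while it sounds; the debug flag is left on.
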SoundCue /JethroWeaponGameplay/Assets/Audio/ClearShell/IP_SC_Jethro_ClearShell_Overheat
root=Mixer_1
nodes [
 {id=Delay_0, children [WavePlayer_6], properties {delayMax=2.33, delayMin=2.33}}
 {id=Mixer_1, children [WavePlayer_5,Delay_0], properties {inputVolume [1,0.7]}}
 {id=WavePlayer_5, properties {soundWaveAssetPtr=&/IP_Jethro_ClearShell_Overheat_01}}
 {id=WavePlayer_6, properties {soundWaveAssetPtr=&/IP_Jethro_ClearShell_Overheat_End_01}}
]
properties {
 attenuationSettings=&Fort_Weapon_Reload_Att
 bDebug=true
 bOverrideConcurrency=true
 maxDistance=2250
 soundClassObject=&Fort_WeaponFoley
 volumeMultiplier=1
 concurrencyOverrides {bLimitToOwner=true, maxCount=1, resolutionRule=PreventNew}
}
