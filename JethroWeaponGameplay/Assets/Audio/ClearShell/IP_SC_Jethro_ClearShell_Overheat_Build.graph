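# Sound cue for the Jethro ClearShell weapon's overheat build-up audio.
# A Random node picks one of three overheat build wave variations with equal weights.
# It routes through a Fortnite ranged-weapon low-ammo degraded sound class and overrides concurrency to prevent new instances while at the limit.
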SoundCue /JethroWeaponGameplay/Assets/Audio/ClearShell/IP_SC_Jethro_ClearShell_Overheat_Build
root=Random_1
nodes [
 {id=Random_1, children [WavePlayer_1,WavePlayer_5,WavePlayer_6], properties {weights [1,1,1]}}
 {id=WavePlayer_1, properties {soundWaveAssetPtr=&/IP_Jethro_ClearShell_Overheat_Build_01}}
 {id=WavePlayer_5, properties {soundWaveAssetPtr=&/IP_Jethro_ClearShell_Overheat_Build_02}}
 {id=WavePlayer_6, properties {soundWaveAssetPtr=&/IP_Jethro_ClearShell_Overheat_Build_03}}
]
properties {
 bOverrideConcurrency=true
 maxDistance=2097152
 soundClassObject=/Game/Sounds/FortSoundClassesAndModes/Fort_WeaponRanged_LowAmmo_Degraded_PSM
 volumeMultiplier=1
 concurrencyOverrides {resolutionRule=PreventNew}
}
